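# Cliff material for Artemis environment meshes, blending world-aligned rock sides with sand tops by vertex normal and texture height, with simpler mobile and forward variants.
# Static switches add optional baked diffuse, normals and masks, and a detail texture layers over the rock.
# Landscape runtime virtual textures height-blend color and normals near the ground.
# The result feeds shared snow progression, then hit-effect world position offset driven by custom primitive data.
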Material /Game/Athena/Artemis/Environments/Landscape/Cliffs/M_TempCliffs
graph {
 properties {
  bTangentSpaceNormal=false
  bUseMaterialAttributes=true
  bUsedWithInstancedStaticMeshes=true
  bUsedWithNanite=true
  bUsedWithSplineMeshes=true
  pixelDepthOffsetMode=PDOM_Legacy
  usedShadingModels=Default Lit
 }
 nodes [
  n1 = NamedRerouteUsage()
  n2 = StaticSwitchParameter(True=NamedRerouteUsage(), False=@n1) {defaultValue=true, group=RVT, parameterName=UseRVT}
  n3 = MaterialFunctionCall!(
   IsStaticSnow
   UseAlternateStaticSnowMask
   AlternateStaticSnowMask
   "UseRoad&&OcclusionMask"=StaticBool()
   IsForGrassMeshToPushDown
   AdditionalMaskToRemoveSnow
   IsFallingSnowFX
   UseMinimapCoords
   IncreaseSnowUsingSlope
   UseSmoothBorder
   LerpHighM
   TemporalHeightLerpContrast
   IsLobbyIsland
   UseBR-IslandMask
   IsTree
   IsForRenderTarget
   UseSparkle
   UseSnowNormals
   UseAdditionalMaskToRemoveSnow
   IsTangentSpaceInput
   UseVertexNormalsForBasicSnow
   UseTangentSnowNormals
   Material=ShadingPathSwitch(Default=@n2, Deferred=@n2, Forward=@n1, Mobile=@n1)
  ) {materialFunction=/Game/Athena/Apollo/Environments/Landscape/Materials/Biome/UniqueBiomeMaterialFunctions/MF_Apollo_Snow_Progression}
  MaterialFunctionCall!(In=@n3.Material) {materialFunction=/Game/Athena/WinterArt/Materials/MF_GameWideShared}
  n4 = MaterialFunctionCall!(
   WSDeformationCenter=VectorParameter() {bUseCustomPrimitiveData=true, group=z_code, parameterName=WSDeformationCenter, primitiveDataIndex=20}.RGB
   Code_DestructionAmount=ScalarParameter() {group=z_code, parameterName=Code_DestructionAmount}
   Code_UseUvForDamage
   Code_UseHitEffects=StaticBoolParameter() {defaultValue=true, group=z_code, parameterName=Code_UseHitEffects}
   Code_Emissive
   WSDamageRadius=ScalarParameter() {bUseCustomPrimitiveData=true, defaultValue=1, group=z_code, parameterName=WSDamageRadius, primitiveDataIndex=26}
   CrackDamageShader
   Diffuse
   ScalarModDamageCracks
   ScalarOffsetforDamageCracks
   ExposeLerpFunctionForInvertingTheHitDistanceCalc
   InvertDamageFalloff
   CrackWorldPosition
   WorldPosition
   DistanceCullFadeAnimationOffset=ScalarParameter() {bUseCustomPrimitiveData=true, group=z_code, parameterName=DistanceCullFadeAnimationOffset, primitiveDataIndex=12}
  ) {materialFunction=/Game/Packages/jl_tests/Fortnite_HitEffects}
  n5 = StaticSwitchParameter(
   True=MaterialFunctionCall!(Material=@n3.Material, WorldPositionOffset=@n4.Hit And Shrink WPO) {materialFunction=/Engine/Functions/MaterialLayerFunctions/MatLayerBlend_AddWorldPositionOffset}.Blended Material
   False=SetMaterialAttributes!(MaterialAttributes=@n3.Material, World Position Offset=@n4.Hit And Shrink WPO) {attributeSetTypes [#1]}
  ) {parameterName=Use Additive WPO}
  MaterialProxyReplace(
   Realtime=@n5
   MaterialProxy=SetMaterialAttributes!(
    MaterialAttributes=GetMaterialAttributes(None=@n5) {attributeGetTypes [#2]}.MaterialAttributes
    Normal=Constant3Vector() {constant=<rgba 0 0 1 0>}
   ) {attributeSetTypes [#2]}
  )
  n6 = MaterialFunctionCall!(CustomSeaLevel, BypassBiomeColors, UseOceanLevel) {materialFunction=/Game/Athena/Apollo/Environments/Landscape/Materials/Biome/CommonMaterials/Mobile/MF_Mobile_Common_Sand}
  n7 = TextureObjectParameter() {group=Top, parameterName=Diffuse Top, sortPriority=4, texture=/Game/Environments/Apollo/Sets/Terrain/Textures/T_Apollo_Sand_D_01}
  n8 = TextureObjectParameter() {
   group=Top
   parameterName=Spec Top
   samplerType=SAMPLERTYPE_Masks
   sortPriority=4
   texture=/Game/Environments/Apollo/Sets/Terrain/Textures/T_Apollo_Sand_S_01
  }
  n9 = TextureObjectParameter() {
   group=Top
   parameterName=Normals Top
   samplerType=SAMPLERTYPE_Normal
   sortPriority=6
   texture=/Game/Environments/Apollo/Sets/Terrain/Textures/T_Apollo_Sand_N_01
  }
  n10 = MaterialFunctionCall!(
   Diffuse Texture Object=@n7
   Spec Texture Object=@n8
   TextureSize=ScalarParameter() {defaultValue=1000, group=Top, parameterName=Scale Top, sortPriority=3}
   --------------------------
   Use seperate settings for Z projection
   Use seperate texture for Z projection
   Texture Size Z projection
   Z Plane Diffuse Texture Object
   Z Plane Spec Texture Object
   "--------------------------      "
   ProjectionTransitionContrast
   Use Texture Alpha Heightmap to Refine the Transistion
   HeightLerpTransitionContrast
   PickTheDiffuseHeightChannel
   TheDiffuseHeightChannel
   "--------------------------       "
   NormalTextureObject=@n9
   World Space Normals
   "--------------------------  "
   Export Float 4 Diffuse=StaticBool() {value=true}
   Export Float 4 Spec
   "--------------------------     "
   WorldPosition
   World Space Normal
  ) {materialFunction=/Engine/Functions/Engine_MaterialFunctions02/Texturing/WorldAlignedTextures_Complex}
  NamedRerouteDeclaration(None=StaticSwitchParameter(True=@n6.Height, False=ComponentMask(None=@n10.Z Diffuse Texture) {a=true}) {defaultValue=true, parameterName=UseDefaultSandForTop}) {name=Top Height, nodeColor=<rgba 0 1 0.1562 1>}
  n11 = StaticSwitchParameter(True=NamedRerouteUsage(), False=VertexNormalWS()) {defaultValue=true, group=Baked Textures, parameterName=UseBakedMasks}
  NamedRerouteDeclaration(None=MaterialFunctionCall(In=ComponentMask(None=@n11) {b=true}, Contrast=ScalarParameter() {group=RVT, parameterName=UpSideContrast}) {materialFunction=&CheapContrast}.Result) {name=UpSideMask, nodeColor=<rgba 0 1 0.1094 1>}
  n12 = TextureSampleParameter2D() {group=Baked Textures, parameterName=Normals, samplerType=SAMPLERTYPE_Normal, texture=/Game/Athena/WinterArt/Textures/Cracks_N}
  NamedRerouteDeclaration(None=Transform(None=@n12.RGB)) {name=Baked N, nodeColor=<rgba 0.125 1 0 1>}
  n13 = VectorParameter() {group=Ground Edge Dust, parameterName=Ground Edge Dust, defaultValue=<rgba 0.4809 0.1338 0.0827 0>}
  n14 = GetMaterialAttributes(None=NamedRerouteUsage()) {attributeGetTypes [#3]}
  n15 = Saturate(None=Divide(A=NamedRerouteUsage(), B=ScalarParameter() {defaultValue=200, group=Ground Edge Dust, parameterName=Ground Edge Dust}))
  NamedRerouteDeclaration(None=SetMaterialAttributes!(MaterialAttributes=@n14.MaterialAttributes, Base Color=LinearInterpolate(A=@n13.RGB, B=@n14.Base Color, Alpha=@n15)) {attributeSetTypes [#3]}) {name=Dust To Side, nodeColor=<rgba 0.0312 1 0 1>}
  n16 = TextureSampleParameter2D() {
   group=Baked Textures
   parameterName=Baked Diffuse
   samplerType=SAMPLERTYPE_LinearColor
   texture=/Game/Global/Textures/Default/Blanks/T_BlackLinearColor
  }
  n17 = TextureSampleParameter2D() {
   group=Baked Textures
   parameterName=Baked Spec
   samplerType=SAMPLERTYPE_Masks
   texture=/Game/Global/Textures/Default/Blanks/Temp_specular_Mask
  }
  n18 = SetMaterialAttributes!(MaterialAttributes, Base Color=@n16.RGB, Roughness=@n17.B, Specular=@n17.B, Metallic=@n17.B) {attributeSetTypes [
    #3
    #4
    #5
    #6
   ]}
  n19 = TextureObjectParameter() {group=Side, parameterName=Diffuse, sortPriority=4, texture=/Game/Environments/Apollo/Sets/Terrain/Textures/T_Apollo_Cliff_D_01}
  n20 = TextureObjectParameter() {
   group=Side
   parameterName=Spec Side
   samplerType=SAMPLERTYPE_Masks
   sortPriority=5
   texture=/Game/Environments/Apollo/Sets/Terrain/Textures/T_Apollo_Cliff_S_01
  }
  n21 = TextureObjectParameter() {
   group=Side
   parameterName=Normals Side
   samplerType=SAMPLERTYPE_Normal
   sortPriority=6
   texture=/Game/Environments/Apollo/Sets/Terrain/Textures/T_Apollo_Cliff_N_01
  }
  n22 = MaterialFunctionCall(
   Diffuse Texture Object=@n19
   Spec Texture Object=@n20
   TextureSize=ScalarParameter() {defaultValue=512, group=Side, parameterName=Scale Side, sortPriority=3}
   NormalTextureObject=@n21
   Export Float 4 Diffuse=StaticBool() {value=true}
  ) {materialFunction=&WorldAlignedTextures_Complex}
  n23 = VectorParameter() {group=Side, parameterName=ColorTone B, sortPriority=1, defaultValue=<rgba 0.5 0.5 0.5 0>}
  n24 = VectorParameter() {group=Side, parameterName=ColorTone A, sortPriority=0, defaultValue=<rgba 0.5 0.5 0.5 0>}
  n25 = ScalarParameter() {defaultValue=0.5, group=Side, parameterName=ColorViaAlpha, sortPriority=2}
  n26 = NamedRerouteDeclaration(None=ComponentMask(None=@n22.XY Diffuse Texture) {a=true}) {name=Side Height, nodeColor=<rgba 0.7109 1 0 1>}
  n27 = ScalarParameter() {defaultValue=3, group=Side, parameterName=ColorViaConrast, sortPriority=2}
  n28 = MaterialFunctionCall(Float3=@n22.XY Spec) {materialFunction=&BreakOutFloat3Components}
  n29 = SetMaterialAttributes!(
   MaterialAttributes
   Base Color=MaterialFunctionCall(
    Base=ComponentMask(None=@n22.XY Diffuse Texture) {b=true, g=true, r=true}
    Blend=MaterialFunctionCall(A=@n23.RGB, B=@n24.RGB, Transition Phase=@n25, Height Texture=@n26, Contrast=@n27) {materialFunction=&HeightLerp}.Results
   ) {materialFunction=&Blend_Overlay}.Result
   Specular=Multiply(A=ScalarParameter() {defaultValue=1, group=Side, parameterName=Side Spec Tweak}, B=@n28.R)
   Roughness=@n28.B
   Normal=@n22.Normal XY Texture
  ) {attributeSetTypes [
    #3
    #5
    #4
    #2
   ]}
  n30 = Multiply(A=ComponentMask(None=@n22.XY Diffuse Texture) {r=true}, B=Constant() {r=2.25})
  n31 = Constant() {r=-0.3}
  n32 = SetMaterialAttributes(
   Base Color=@n22.XY Diffuse Texture
   Specular=Saturate(None=Add(A=@n30, B=@n31))
   Roughness=Constant() {r=0.85}
   Normal=@n22.Normal XY Texture
  ) {attributeSetTypes [
    #3
    #5
    #4
    #2
   ]}
  NamedRerouteDeclaration(None=StaticSwitchParameter(True=@n18, False=ShadingPathSwitch(Default=@n29, Forward=@n32, Mobile=@n32)) {parameterName=UseBakedDiffuseandSpec}) {name=Side, nodeColor=<rgba 1 0.3516 0 1>}
  n33 = ComponentMask(None=@n10.Z Diffuse Texture) {b=true, g=true, r=true}
  n34 = MaterialFunctionCall(Float3=@n10.Z Spec) {materialFunction=&BreakOutFloat3Components}
  n35 = SetMaterialAttributes(Base Color=@n33, Specular=@n34.R, Roughness=@n34.R, Normal=@n10.Normal Z Texture) {attributeSetTypes [
    #3
    #5
    #4
    #2
   ]}
  n36 = SetMaterialAttributes(Base Color=@n33, Specular=Constant() {r=0.3}, Roughness=Constant() {r=0.9}, Normal=VertexNormalWS()) {attributeSetTypes [
    #3
    #5
    #4
    #2
   ]}
  NamedRerouteDeclaration(None=StaticSwitchParameter(True=@n6.ForStaticmesh, False=ShadingPathSwitch(Default=@n35, Forward=@n36, Mobile=@n36)) {defaultValue=true, parameterName=UseDefaultSandForTop}) {name=Top, nodeColor=<rgba 0.4062 0 1 1>}
  n37 = GetMaterialAttributes(None=NamedRerouteUsage()) {attributeGetTypes [
    #3
    #5
    #4
    #2
   ]}
  n38 = TextureObjectParameter() {
   group=Detail Texture
   parameterName=Detail D
   sortPriority=1
   texture=/Game/Environments/Apollo/Rocks/Natural_Rocks/Textures/T_RockDetails_02D
  }
  n39 = MaterialFunctionCall(TextureObject=@n38, TextureSize=ScalarParameter() {defaultValue=768, group=Detail Texture, parameterName=DetailTile}) {materialFunction=/Game/Weapons/WeaponSkins/Wraps/MaterialFunctions/WorldAlignedTexture}
  n40 = ComponentMask(None=StaticSwitch(True=@n39.Z Texture, False=@n39.Z Texture, Value=StaticBoolParameter() {group=Detail Texture, parameterName=Use Z only projection})) {r=true}
  n41 = Multiply(A=ScalarParameter() {defaultValue=1, group=Detail Texture, parameterName=Details Tone}, B=@n40)
  n42 = LinearInterpolate(
   A=ScalarParameter() {defaultValue=-0.773, group=Detail Texture, parameterName=Detail R Min, sortPriority=5}
   B=ScalarParameter() {defaultValue=1.8112, group=Detail Texture, parameterName=Detail R Max, sortPriority=6}
   Alpha=@n40
  )
  n43 = SetMaterialAttributes(
   MaterialAttributes=@n37.MaterialAttributes
   Base Color=Saturate(None=MaterialFunctionCall(Base=@n37.Base Color, Blend=@n41) {materialFunction=&Blend_Overlay}.Result)
   Specular=Saturate(None=Multiply(A=@n40, B=@n37.Specular))
   Roughness=Saturate(None=Multiply(A=@n37.Roughness, B=@n42))
  ) {attributeSetTypes [
    #3
    #5
    #4
    #2
   ]}
  NamedRerouteDeclaration(None=BlendMaterialAttributes(A=@n37.MaterialAttributes, B=@n43, Alpha=ScalarParameter() {defaultValue=1, group=Detail Texture, parameterName=Details Visibility})) {name=Details On Side, nodeColor=<rgba 0.2187 1 0 1>}
  n44 = NamedRerouteUsage()
  n45 = NamedRerouteUsage()
  n46 = VertexNormalWS()
  n47 = NamedRerouteUsage()
  n48 = BlendMaterialAttributes(
   A=ShadingPathSwitch(Default=@n44, Deferred=@n44, Forward=@n45, Mobile=@n45)
   B=NamedRerouteUsage()
   Alpha=MaterialFunctionCall(Transition Phase=ComponentMask(None=@n46) {b=true}, Height Texture=OneMinus(None=@n47)) {materialFunction=&HeightLerp}.Lerp Alpha No Contrast
  )
  NamedRerouteDeclaration(None=StaticSwitchParameter(True=NamedRerouteUsage(), False=NamedRerouteDeclaration(None=@n48) {name=PreBlend, nodeColor=<rgba 0 0.9766 1 1>}) {group=Baked Textures, parameterName=Use Baked Textures}) {name=Blended, nodeColor=<rgba 0 1 0.9297 1>}
  n49 = NamedRerouteUsage()
  n50 = NamedRerouteUsage()
  n51 = RuntimeVirtualTextureSampleParameter() {
   group=RVT
   materialType=BaseColor_Normal_Specular_YCoCg
   parameterName=RVT
   sortPriority=5
   virtualTexture=/Game/Athena/Apollo/Maps/Landscape/VT/RVT_Apollo_Landscape
  }
  n52 = WorldPosition()
  n53 = RuntimeVirtualTextureSampleParameter() {
   group=RVT
   materialType=WorldHeight
   parameterName=RVTHeight
   sortPriority=5
   virtualTexture=/Game/Athena/Apollo/Maps/Landscape/VT/RVT_Apollo_Landscape_Height
  }
  n54 = NamedRerouteDeclaration(None=Subtract(A=ComponentMask(None=@n52.XYZ) {b=true}, B=@n53.WorldHeight)) {name=RVT Height, nodeColor=<rgba 1 0.7969 0 1>}
  n55 = LinearInterpolate(
   A=ScalarParameter() {defaultValue=255, group=RVT, parameterName=Blend RVT Side Normals, sortPriority=-1}
   B=ScalarParameter() {defaultValue=255, group=RVT, parameterName=Blend RVT Top Normals, sortPriority=-1}
   Alpha=NamedRerouteUsage()
  )
  n56 = LinearInterpolate(A=LinearInterpolate(A=Constant() {r=-2}, B=Constant() {r=4}, Alpha=NamedRerouteUsage()), B=NamedRerouteUsage(), Alpha=NamedRerouteUsage())
  n57 = BlendMaterialAttributes(
   A=SetMaterialAttributes!(MaterialAttributes=@n50, Normal=@n51.Normal) {attributeSetTypes [#2]}
   B=@n50
   Alpha=MaterialFunctionCall(Transition Phase=Divide(A=@n54, B=@n55), Height Texture=@n56) {materialFunction=&HeightLerp}.Alpha
  ) {vertexAttributeBlendType=UseA}
  n58 = LinearInterpolate(
   A=ScalarParameter() {defaultValue=555.9241, group=RVT, parameterName=Blend RVT Sides, sortPriority=-2}
   B=ScalarParameter() {defaultValue=100, group=RVT, parameterName=Blend RVT Top, sortPriority=-2}
   Alpha=NamedRerouteUsage()
  )
  n59 = BlendMaterialAttributes(
   A=SetMaterialAttributes!(MaterialAttributes=@n57, Base Color=@n51.BaseColor, Specular=@n51.BaseColor, Roughness=@n51.BaseColor) {attributeSetTypes [#3, #5, #4]}
   B=@n57
   Alpha=MaterialFunctionCall(Transition Phase=Divide(A=@n54, B=@n58), Height Texture=@n56) {materialFunction=&HeightLerp}.Lerp Alpha No Contrast
  ) {vertexAttributeBlendType=UseA}
  NamedRerouteDeclaration(None=MaterialFunctionCall(NoRVT=@n49, RVT=MaterialProxyReplace(Realtime=@n59, MaterialProxy=@n49)) {materialFunction=&FN_RVT_Switch_Material}.Result) {name=Blended+RVT, nodeColor=<rgba 0.6641 0 1 1>}
  n60 = NamedRerouteUsage()
  n61 = GetMaterialAttributes(None=@n60) {attributeGetTypes [#2, #3]}
  n62 = Multiply(
   A=Transform(None=@n61.Normal) {transformSourceType=TRANSFORMSOURCE_World, transformType=TRANSFORM_Tangent}
   B=Constant3Vector() {constant=<rgba 1 1 0 0>}
  )
  n63 = ScalarParameter() {defaultValue=1, group=Baked Textures, parameterName=Tiled Normals Value}
  n64 = Transform(None=Add(A=Multiply(A=@n62, B=@n63), B=@n12.RGB))
  n65 = TextureSampleParameter2D() {group=Baked Textures, parameterName=BakedMasks, samplerType=SAMPLERTYPE_Masks, texture=/Game/Athena/WinterArt/Textures/Cracks_Mask}
  n66 = LinearInterpolate(
   A=@n61.Base Color
   B=MaterialFunctionCall(Base=@n61.Base Color, Blend=@n65.R) {materialFunction=&Blend_Overlay}.Result
   Alpha=ScalarParameter() {defaultValue=1, group=Baked Textures, parameterName=Curvature}
  )
  n67 = GetMaterialAttributes(None=@n60) {attributeGetTypes [#2]}
  n68 = Multiply(
   A=Transform(None=@n67.Normal) {transformSourceType=TRANSFORMSOURCE_World, transformType=TRANSFORM_Tangent}
   B=Constant3Vector() {constant=<rgba 1 1 0 0>}
  )
  n69 = Transform(None=Add(A=Multiply(A=@n68, B=@n63), B=@n12.RGB))
  NamedRerouteDeclaration(None=StaticSwitchParameter(
    True=SetMaterialAttributes!(MaterialAttributes=@n61.MaterialAttributes, Normal=@n64, Base Color=@n66) {attributeSetTypes [#2, #3]}
    False=SetMaterialAttributes!(MaterialAttributes=@n67.MaterialAttributes, Normal=@n69) {attributeSetTypes [#2]}
   ) {defaultValue=true, group=Baked Textures, parameterName=UseBakedMasks}) {name=Baked Normal, nodeColor=<rgba 0.1719 0 1 1>}
  n70 = NamedRerouteDeclaration(None=@n29) {name=Side To dust, nodeColor=<rgba 0 1 0.3203 1>}
  StaticSwitchParameter(
   True=@n70
   False=MaterialFunctionCall(NoRVT=@n70, RVT=NamedRerouteUsage()) {materialFunction=&FN_RVT_Switch_Material}.Result
  ) {group=Ground Edge Dust, parameterName=SkipDust}
 ]
}
